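# Physics asset for the long belt on the F_MED_Donut_Plate female medium player body cosmetic, set up for dangling belt motion.
# Eight box bodies form a constrained chain from dyn_belt_parent, with tight swing and twist limits at the root and looser, uniform limits below.
# Capsules on the pelvis, thighs, calves and feet provide lower-body shapes alongside the belt.
# The small dyn_belt_base box has no constraint.
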PhysicsAsset /BRCosmetics/Characters/Player/Female/Medium/Bodies/F_MED_Donut_Plate/Meshes/F_MED_Donut_Plate_LongBelt_Physics
bodies {
 calf_l [{shapeType=Capsule, length=35.659, radius=12, center=<-23.0141 -3.4174 0>, rotation=<rot 90 0 0>}]
 calf_r [{shapeType=Capsule, length=49.3746, radius=12, center=<19.4762 3.3474 0.0001>, rotation=<rot 90 0 0>}]
 dyn_belt_base [{shapeType=Box, extentX=0.5, extentY=0.5, extentZ=0.5}]
 dyn_belt_long_01 [{shapeType=Box, extentX=3.7419, extentY=8.4963, extentZ=2.5, center=<1.5473 0 -0.6128>}]
 dyn_belt_long_02 [{shapeType=Box, extentX=5.5389, extentY=8.0622, extentZ=1.75, center=<2.828 0 0.2965>}]
 dyn_belt_long_03 [{shapeType=Box, extentX=8.1502, extentY=9.385, extentZ=2.8496, center=<3.6478 0 0.0001>}]
 dyn_belt_long_04 [{shapeType=Box, extentX=13.4002, extentY=15.385, extentZ=2.8496, center=<6.8747 0 0>}]
 dyn_belt_long_05 [{shapeType=Box, extentX=13.4002, extentY=15.385, extentZ=2.8496, center=<6.8747 0 0>}]
 dyn_belt_long_06 [{shapeType=Box, extentX=13.4002, extentY=15.385, extentZ=2.8496, center=<6.8747 0 0>}]
 dyn_belt_long_07 [{shapeType=Box, extentX=13.4002, extentY=15.385, extentZ=2.8496, center=<6.8747 0 0>}]
 dyn_belt_long_08 [{shapeType=Box, extentX=13.4002, extentY=15.385, extentZ=2.8496, center=<6.8747 0 0>}]
 dyn_belt_parent [{shapeType=Box, extentX=3.8438, extentY=1.4922, extentZ=4.5493, center=<1.5771 0.1892 0.2761>, rotation=<rot 0 88.1322 -74.7806>}]
 foot_l [{shapeType=Capsule, length=14.8659, radius=6.1834, center=<-4.3935 6.7675 0.1079>, rotation=<rot 0 0 90>}]
 foot_r [{shapeType=Capsule, length=14.8659, radius=6.1834, center=<4.3166 -6.1931 0.1077>, rotation=<rot 0 0 90>}]
 pelvis [{shapeType=Capsule, length=14.75, radius=9.25, center=<-5.2731 -2.5927 0>}]
 thigh_l [{shapeType=Capsule, length=33.0656, radius=8.9344, center=<-22.5539 0.0002 0>, rotation=<rot 90 0 0>}]
 thigh_r [{shapeType=Capsule, length=31.9798, radius=10.0202, center=<22.7634 0.0001 0>, rotation=<rot 90 0 0>}]
}
constraints [
 {bone1Name=dyn_belt_long_01, bone2Name=dyn_belt_parent, swing1LimitDegrees=10, swing2LimitDegrees=5, twistLimitDegrees=0.5}
 {bone1Name=dyn_belt_long_02, bone2Name=dyn_belt_long_01, swing1LimitDegrees=20, swing2LimitDegrees=10, twistLimitDegrees=5}
 {bone1Name=dyn_belt_long_03, bone2Name=dyn_belt_long_02, swing1LimitDegrees=20, swing2LimitDegrees=10, twistLimitDegrees=5}
 {bone1Name=dyn_belt_long_04, bone2Name=dyn_belt_long_03, swing1LimitDegrees=20, swing2LimitDegrees=10, twistLimitDegrees=5}
 {bone1Name=dyn_belt_long_05, bone2Name=dyn_belt_long_04, swing1LimitDegrees=20, swing2LimitDegrees=10, twistLimitDegrees=5}
 {bone1Name=dyn_belt_long_06, bone2Name=dyn_belt_long_05, swing1LimitDegrees=20, swing2LimitDegrees=10, twistLimitDegrees=5}
 {bone1Name=dyn_belt_long_07, bone2Name=dyn_belt_long_06, swing1LimitDegrees=20, swing2LimitDegrees=10, twistLimitDegrees=5}
 {bone1Name=dyn_belt_long_08, bone2Name=dyn_belt_long_07, swing1LimitDegrees=20, swing2LimitDegrees=10, twistLimitDegrees=5}
]
