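# Physics asset for the short belt on the female medium Donut Plate player cosmetic body, defining the shapes and joints for its dangling motion.
# A chain of box bodies runs from the belt base through eight segments, with swing limits tightening toward the tip and uniformly narrow twist.
# Unconstrained capsules on the pelvis, thighs, calves, left foot and right ball give the belt body shapes to collide with.
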PhysicsAsset /BRCosmetics/Characters/Player/Female/Medium/Bodies/F_MED_Donut_Plate/Meshes/F_MED_Donut_Plate_ShortBelt_Physics
bodies {
 ball_r [{shapeType=Capsule, length=13.7319, radius=7.6031, center=<4.4021 0 0.0001>, rotation=<rot 90 0 0>}]
 calf_l [{shapeType=Capsule, length=38.8511, radius=12, center=<-21.7059 -3.4174 0.0001>, rotation=<rot 90 0 0>}]
 calf_r [{shapeType=Capsule, length=39.2518, radius=12, center=<21.3785 3.5163 0.4264>, rotation=<rot 90 0 0>}]
 dyn_belt_base [{shapeType=Box, extentX=0.5, extentY=0.5, extentZ=0.5}]
 dyn_belt_parent [{shapeType=Box, extentX=3.7419, extentY=8.4963, extentZ=1.5, center=<1.5473 0 0>}]
 dyn_belt_short_01 [{shapeType=Box, extentX=4.5389, extentY=8.0622, extentZ=2.25, center=<1.4455 0 -0.0001>}]
 dyn_belt_short_02 [{shapeType=Box, extentX=8.1502, extentY=9.385, extentZ=2.8496, center=<3.6478 0 0.0001>}]
 dyn_belt_short_03 [{shapeType=Box, extentX=13.4002, extentY=15.385, extentZ=2.8496, center=<6.8747 0 0.0002>}]
 dyn_belt_short_04 [{shapeType=Box, extentX=13.4002, extentY=15.385, extentZ=2.8496, center=<6.8747 0 0.0002>}]
 dyn_belt_short_05 [{shapeType=Box, extentX=13.4002, extentY=15.385, extentZ=2.8496, center=<6.8747 0 0.0002>}]
 dyn_belt_short_06 [{shapeType=Box, extentX=13.4002, extentY=15.385, extentZ=2.8496, center=<6.8747 0 0.0002>}]
 dyn_belt_short_07 [{shapeType=Box, extentX=13.4002, extentY=15.385, extentZ=2.8496, center=<6.8747 0 0.0002>}]
 dyn_belt_short_08 [{shapeType=Box, extentX=13.4002, extentY=15.385, extentZ=2.8496, center=<6.8747 0 0.0002>}]
 foot_l [{shapeType=Capsule, length=13.7319, radius=7.6031, center=<-2.8591 6.1877 0.0004>, rotation=<rot 0 0 -90>}]
 pelvis [{shapeType=Capsule, length=14.75, radius=9.25, center=<-6.8209 -2.1754 0>}]
 thigh_l [{shapeType=Capsule, length=33.3558, radius=8.6442, center=<-22.5539 0.0002 0>, rotation=<rot 90 0 0>}]
 thigh_r [{shapeType=Capsule, length=32.8536, radius=9.1463, center=<22.1135 0.901 0>, rotation=<rot 90 0 0>}]
}
constraints [
 {bone1Name=dyn_belt_parent, bone2Name=dyn_belt_base, swing1LimitDegrees=30, swing2LimitDegrees=20, twistLimitDegrees=5}
 {bone1Name=dyn_belt_short_01, bone2Name=dyn_belt_parent, swing1LimitDegrees=30, swing2LimitDegrees=20, twistLimitDegrees=5}
 {bone1Name=dyn_belt_short_02, bone2Name=dyn_belt_short_01, swing1LimitDegrees=30, swing2LimitDegrees=20, twistLimitDegrees=5}
 {bone1Name=dyn_belt_short_03, bone2Name=dyn_belt_short_02, swing1LimitDegrees=20, swing2LimitDegrees=15, twistLimitDegrees=5}
 {bone1Name=dyn_belt_short_04, bone2Name=dyn_belt_short_03, swing1LimitDegrees=20, swing2LimitDegrees=15, twistLimitDegrees=5}
 {bone1Name=dyn_belt_short_05, bone2Name=dyn_belt_short_04, swing1LimitDegrees=20, swing2LimitDegrees=15, twistLimitDegrees=5}
 {bone1Name=dyn_belt_short_06, bone2Name=dyn_belt_short_05, swing1LimitDegrees=10, swing2LimitDegrees=5, twistLimitDegrees=5}
 {bone1Name=dyn_belt_short_07, bone2Name=dyn_belt_short_06, swing1LimitDegrees=10, swing2LimitDegrees=5, twistLimitDegrees=5}
 {bone1Name=dyn_belt_short_08, bone2Name=dyn_belt_short_07, swing1LimitDegrees=10, swing2LimitDegrees=5, twistLimitDegrees=5}
]
