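# Animation Blueprint for the F_MED_Donut_Plate body cosmetic, driving secondary motion on its short and long belts.
# It copies the parent mesh pose, runs RigidBody per belt, then caches the result.
# When rigid body is disabled or LOD exceeds 1, output switches from small bone offsets to AnimDynamics chains with corrective rotations, both atop that cache.
# Skydive states halve RigidBody alpha and roll the belt parent bone.
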AnimBlueprint /BRCosmetics/Characters/Player/Female/Medium/Bodies/F_MED_Donut_Plate/Meshes/F_MED_Donut_Plate_AnimBp
target_skeleton=&/F_MED_Donut_Plate_Skeleton
graphs [
 AnimGraph {nodes [
   n1 = "Transform (Modify) Bone - Bone: dyn_belt_base"(
    ComponentPose=Local To Component(LocalPose=Use cached pose 'BeltCache'())
    Alpha=1
    Translation="0,-1.45,0"
    Rotation="0,0.15,0"
    Scale="1,1,1"
   ) {
    rotationMode=BMM_Additive
    rotationSpace=BCS_ParentBoneSpace
    translationMode=BMM_Additive
    translationSpace=BCS_ParentBoneSpace
    alphaBoolBlend {blendOption=Linear}
    alphaScaleBiasClamp {bInterpResult=true, interpSpeedDecreasing=5, interpSpeedIncreasing=6}
    boneToModify {boneName=dyn_belt_base}
    rotation {yaw=0.15}
    translation {y=-1.45}
   }
   n2 = "Transform (Modify) Bone - Bone: dyn_belt_short_01"(ComponentPose=@n1, Alpha=1, Translation="0,0.25,0", Rotation="-5,0,0", Scale="1,1,1") {
    rotationMode=BMM_Additive
    rotationSpace=BCS_ParentBoneSpace
    translationMode=BMM_Additive
    translationSpace=BCS_ParentBoneSpace
    alphaBoolBlend {blendOption=Linear}
    alphaScaleBiasClamp {bInterpResult=true, interpSpeedDecreasing=5, interpSpeedIncreasing=6}
    boneToModify {boneName=dyn_belt_short_01}
    rotation {pitch=-5}
    translation {y=0.25}
   }
   n3 = "Transform (Modify) Bone - Bone: dyn_belt_long_01"(ComponentPose=@n2, Alpha=1, Translation="0,0,0.25", Rotation="3,0,0", Scale="1,1,1") {
    rotationMode=BMM_Additive
    rotationSpace=BCS_ParentBoneSpace
    translationMode=BMM_Additive
    translationSpace=BCS_ParentBoneSpace
    alphaBoolBlend {blendOption=Linear}
    alphaScaleBiasClamp {bInterpResult=true, interpSpeedDecreasing=5, interpSpeedIncreasing=6}
    boneToModify {boneName=dyn_belt_long_01}
    rotation {pitch=3}
    translation {z=0.25}
   }
   n4 = "Anim Dynamics - Chain: dyn_belt_long_01 -> dyn_belt_long_08"(ComponentPose=Local To Component(LocalPose=Use cached pose 'BeltCache'()), Alpha=1, ExternalForce="0,0,0") {
    angularDampingOverride=0.999
    bChain=true
    bGravityOverrideInSimSpace=true
    bOverrideLinearDamping=true
    gravityScale=0.2
    linearDampingOverride=0.9999
    numSolverIterationsPostUpdate=10
    numSolverIterationsPreUpdate=10
    alphaBoolBlend {blendOption=Linear}
    boundBone {boneName=dyn_belt_long_01}
    chainEnd {boneName=dyn_belt_long_08}
    componentAppliedLinearAccClamp=<600000 600000 600000>
    componentLinearAccScale=<0.25 0.25 0.25>
    componentLinearVelScale=<0.25 0.25 0.25>
    gravityOverride {z=10}
    physicsBodyDefinitions [{
      sphereCollisionRadius=0
      boundBone {boneName=dyn_belt_long_01}
      boxExtents=<20 25 5>
      constraintSetup {angularLimitsMax=<0 10 10>, angularLimitsMin=<0 -10 -10>, angularTarget=<0 0 0>}
      localJointOffset=<10 0 0>
     }]
    relativeSpaceBone {boneName=root}
   }
   n5 = "Anim Dynamics - Chain: dyn_belt_short_01 -> dyn_belt_short_08"(ComponentPose=@n4, Alpha=1, ExternalForce="0,0,0") {
    angularDampingOverride=0.999
    bChain=true
    bGravityOverrideInSimSpace=true
    bOverrideLinearDamping=true
    gravityScale=0.2
    linearDampingOverride=0.9999
    numSolverIterationsPostUpdate=10
    numSolverIterationsPreUpdate=10
    alphaBoolBlend {blendOption=Linear}
    boundBone {boneName=dyn_belt_short_01}
    chainEnd {boneName=dyn_belt_short_08}
    componentAppliedLinearAccClamp=<600000 600000 600000>
    componentLinearAccScale=<0.25 0.25 0.25>
    componentLinearVelScale=<0.25 0.25 0.25>
    gravityOverride {z=10}
    physicsBodyDefinitions [{
      sphereCollisionRadius=0
      boundBone {boneName=dyn_belt_short_01}
      boxExtents=<20 25 5>
      constraintSetup {angularLimitsMax=<0 10 10>, angularLimitsMin=<0 -10 -10>, angularTarget=<0 0 0>}
      localJointOffset=<10 0 0>
     }]
    relativeSpaceBone {boneName=root}
   }
   n6 = "Transform (Modify) Bone - Bone: dyn_belt_base"(ComponentPose=@n5, Alpha=1, Translation="0,0,0", Rotation="-14,0,0", Scale="1,1,1") {
    rotationMode=BMM_Additive
    rotationSpace=BCS_BoneSpace
    alphaBoolBlend {blendOption=Linear}
    boneToModify {boneName=dyn_belt_base}
    rotation {pitch=-14}
   }
   n7 = "Transform (Modify) Bone - Bone: dyn_belt_short_02"(ComponentPose=@n6, Alpha=1, Translation="0,0,0", Rotation="-2,0,0", Scale="1,1,1") {
    rotationMode=BMM_Additive
    rotationSpace=BCS_BoneSpace
    alphaBoolBlend {blendOption=Linear}
    boneToModify {boneName=dyn_belt_short_02}
    rotation {pitch=-2}
   }
   Output Pose(Result=Blend Poses by bool(
     BlendPose_0=Component To Local(ComponentPose=@n7)
     BlendPose_1=Component To Local(ComponentPose=@n3)
     bActiveValue=Get isRigidBodyDisabled() -> (isRigidBodyDisabled:bool)
    ) {blendType=Linear}) {layerGroup=None}
   n8 = Get Skydive_Alpha() -> (Skydive_Alpha:double)
   n9 = RigidBody(ComponentPose=Local To Component(LocalPose=Copy Pose From Mesh() {bUseAttachedParent=true}), Alpha=@n8) {
    bClampLinearTranslationLimitToRefPose=true
    bOverrideWorldGravity=true
    overridePhysicsAsset=&/F_MED_Donut_Plate_ShortBelt_Physics
    simulationSpace=BaseBoneSpace
    alphaBoolBlend {blendOption=Linear}
    alphaScaleBiasClamp {bInterpResult=true, interpSpeedDecreasing=6, interpSpeedIncreasing=6}
    baseBoneRef {boneName=root}
    componentAppliedLinearAccClamp=<3000 3000 3000>
    componentLinearAccScale=<0.4 0.3 0.3>
    componentLinearVelScale=<0.4 0.3 0.3>
    overrideWorldGravity {z=-600}
   }
   n10 = RigidBody(ComponentPose=@n9, Alpha=@n8) {
    bClampLinearTranslationLimitToRefPose=true
    bOverrideWorldGravity=true
    overridePhysicsAsset=&/F_MED_Donut_Plate_LongBelt_Physics
    simulationSpace=BaseBoneSpace
    alphaBoolBlend {blendOption=Linear}
    alphaScaleBiasClamp {bInterpResult=true, interpSpeedDecreasing=6, interpSpeedIncreasing=6}
    baseBoneRef {boneName=root}
    componentAppliedLinearAccClamp=<3000 3000 3000>
    componentLinearAccScale=<0.4 0.3 0.3>
    componentLinearVelScale=<0.4 0.3 0.3>
    overrideWorldGravity {z=-600}
   }
   n11 = "Transform (Modify) Bone - Bone: dyn_belt_parent"(ComponentPose=@n10, Alpha=Get Skydive_Rotate_Alpha() -> (Skydive_Rotate_Alpha:double), Translation="0,0,0", Rotation="0,0,45", Scale="1,1,1") {rotationMode=BMM_Additive, alphaBoolBlend {blendOption=Linear}, boneToModify {boneName=dyn_belt_parent}, rotation {roll=45}}
   Save cached pose 'BeltCache'(Pose=Component To Local(ComponentPose=@n11)) {cacheName=BeltCache}
  ]}
 EventGraph {nodes [
   Event BlueprintUpdateAnimation()
   > n1 = Sequence()
   n2 = OR Boolean!(A=Get bIsSkyDiving(), B=Get bIsSkydiveDiveMode(), C:bool=Get bIsSkydiveFloating()) -> (ReturnValue)
   > Set Skydive_Alpha(Skydive_Alpha:double=SelectFloat(A=0.5, B=1, bPickA=@n2)) -> (Output_Get:double)
   > Set Skydive_Rotate_Alpha(Skydive_Rotate_Alpha:double=SelectFloat(A=1, B=0, bPickA=@n2)) -> (Output_Get:double)
   n3 = Get CurrentLOD()
   n4 = Get bRigidBodyEnabled()
   Set isRigidBodyDisabled(
    execute=@n1.then_1
    isRigidBodyDisabled:bool=OR Boolean!(A=NOT Boolean(A=@n4), B="integer > integer"(A=@n3, B=1) {is=title K2Node_CallFunction}) -> (ReturnValue)
   ) -> (Output_Get:bool)
  ]}
]
variables [
 isRigidBodyDisabled {type=boolean}
 Skydive_Alpha {type=number}
 Skydive_Rotate_Alpha {type=number}
]
