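# Physics asset for the bangs face accessory of the F_MED_Donut_Plate female character, giving its hair bones secondary motion.
# Three capsule chains (mid bang, standing bang, back strand) each link through tightly limited swing and twist constraints to a small box base body.
# The mid chain forks at dyn_bang_mid_02 into 03 and 04.
# Head and neck_02 capsules have no constraints, apparently serving as collision volumes for the hair.
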PhysicsAsset /BRCosmetics/Characters/Player/Female/Medium/Bodies/F_MED_Donut_Plate/Meshes/Parts/F_MED_Donut_Plate_FaceAcc_Bangs_Physics
bodies {
 dyn_bang_mid_01 [{shapeType=Capsule, length=2.2297, radius=1.797, center=<2.4395 0 -0.0001>, rotation=<rot 90 0 0>}]
 dyn_bang_mid_02 [{shapeType=Capsule, length=1.7472, radius=1.794, center=<2.0572 0 0>, rotation=<rot 90 0 0>}]
 dyn_bang_mid_03 [{shapeType=Capsule, length=4.6106, radius=0.8428, center=<2.7209 -0.6263 0.8689>, rotation=<rot 77.5061 -179.9998 -162.5462>}]
 dyn_bang_mid_04 [{shapeType=Capsule, length=1.4436, radius=0.5265, center=<0.8705 0 0>, rotation=<rot 90 0 0>}]
 dyn_bang_mid_base [{shapeType=Box, extentX=0.5, extentY=0.5, extentZ=0.5}]
 dyn_bang_mid_parent [{shapeType=Capsule, length=2.2349, radius=1.7959, center=<2.4177 0 -0.0001>, rotation=<rot 90 0 0>}]
 dyn_bang_stand_01 [{shapeType=Capsule, length=2.5218, radius=0.8068, center=<1.2162 0.0001 0>, rotation=<rot 90 150.3033 149.2848>}]
 dyn_bang_stand_02 [{shapeType=Capsule, length=1.9765, radius=0.7031, center=<1.8227 0.0004 0>, rotation=<rot 90 161.4309 160.4124>}]
 dyn_bang_stand_03 [{shapeType=Capsule, length=1.8807, radius=0.625, center=<1.2837 -0.0001 0>, rotation=<rot 90 40.9314 39.9129>}]
 dyn_bang_stand_04 [{shapeType=Capsule, length=1.5053, radius=0.625, center=<0.9405 -0.0002 0>, rotation=<rot 90 0 0>}]
 dyn_bang_stand_base [{shapeType=Box, extentX=0.5, extentY=0.5, extentZ=0.5}]
 dyn_bang_stand_parent [{shapeType=Capsule, length=1.2728, radius=0.9131, center=<1.3002 0 -0.0254>, rotation=<rot 87.0111 0.001 -1.0175>}]
 dyn_strand_back_01 [{shapeType=Capsule, length=2.523, radius=0.7202, center=<1.8411 0 0>, rotation=<rot 90 0 0>}]
 dyn_strand_back_02 [{shapeType=Capsule, length=2.4718, radius=0.5749, center=<1.6377 0 0>, rotation=<rot 90 0 0>}]
 dyn_strand_back_base [{shapeType=Box, extentX=0.5, extentY=0.5, extentZ=0.5}]
 dyn_strand_back_parent [{shapeType=Capsule, length=1.6853, radius=0.7086, center=<1.3381 0 0.0001>, rotation=<rot 90 0 0>}]
 head [
  {shapeType=Capsule, length=3.6437, radius=8.2313, center=<4.6431 0.737 0>, rotation=<rot 90 0 0>}
  {shapeType=Capsule, length=3.0438, radius=3.3917, center=<0.8913 6.7395 -1.7058>}
 ]
 neck_02 [{shapeType=Capsule, length=1.8459, radius=2.6858, center=<8.5758 -0.5844 4.9503>}]
}
constraints [
 {bone1Name=dyn_bang_stand_parent, bone2Name=dyn_bang_stand_base, swing1LimitDegrees=5, swing2LimitDegrees=5, twistLimitDegrees=3}
 {bone1Name=dyn_bang_stand_01, bone2Name=dyn_bang_stand_parent, swing1LimitDegrees=5, swing2LimitDegrees=5, twistLimitDegrees=3}
 {bone1Name=dyn_bang_stand_02, bone2Name=dyn_bang_stand_01, swing1LimitDegrees=5, swing2LimitDegrees=5, twistLimitDegrees=3}
 {bone1Name=dyn_bang_stand_03, bone2Name=dyn_bang_stand_02, swing1LimitDegrees=5, swing2LimitDegrees=5, twistLimitDegrees=3}
 {bone1Name=dyn_bang_stand_04, bone2Name=dyn_bang_stand_03, swing1LimitDegrees=5, swing2LimitDegrees=5, twistLimitDegrees=3}
 {bone1Name=dyn_bang_mid_parent, bone2Name=dyn_bang_mid_base, swing1LimitDegrees=5, swing2LimitDegrees=10, twistLimitDegrees=3}
 {bone1Name=dyn_bang_mid_01, bone2Name=dyn_bang_mid_parent, swing1LimitDegrees=5, swing2LimitDegrees=10, twistLimitDegrees=3}
 {bone1Name=dyn_bang_mid_02, bone2Name=dyn_bang_mid_01, swing1LimitDegrees=10, swing2LimitDegrees=10, twistLimitDegrees=3}
 {bone1Name=dyn_bang_mid_03, bone2Name=dyn_bang_mid_02, swing1LimitDegrees=5, swing2LimitDegrees=10, twistLimitDegrees=3}
 {bone1Name=dyn_bang_mid_04, bone2Name=dyn_bang_mid_02, swing1LimitDegrees=5, swing2LimitDegrees=10, twistLimitDegrees=3}
 {bone1Name=dyn_strand_back_parent, bone2Name=dyn_strand_back_base, swing1LimitDegrees=3, swing2LimitDegrees=2, twistLimitDegrees=2}
 {bone1Name=dyn_strand_back_01, bone2Name=dyn_strand_back_parent, swing1LimitDegrees=3, swing2LimitDegrees=2, twistLimitDegrees=2}
 {bone1Name=dyn_strand_back_02, bone2Name=dyn_strand_back_01, swing1LimitDegrees=3, swing2LimitDegrees=2, twistLimitDegrees=2}
]
