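# Physics asset for the ponytail of the F_MED_Donut_Plate female player cosmetic, with capsule bodies on hair bones and upper body.
# The main eight-segment chain runs from a base and parent bone, with a left strand branching at the parent, right at segment 03 and middle at 04.
# Hair joints have tight swing and twist limits, tightest on the right strand.
# Head, neck, spine and clavicle capsules approximate the body, with only clavicle_r constrained, freely, to spine_05.
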PhysicsAsset /BRCosmetics/Characters/Player/Female/Medium/Bodies/F_MED_Donut_Plate/Meshes/Parts/F_MED_Donut_Plate_Ponytail_FaceAcc_Physics
bodies {
 clavicle_l [{shapeType=Capsule, length=12.3089, radius=7.1386, center=<2.6645 0 -2.3878>, rotation=<rot 90 0 0>}]
 clavicle_r [{shapeType=Capsule, length=12.5589, radius=6.8886, center=<-2.1265 0 2.3878>, rotation=<rot 90 0 180>}]
 dyn_ponytail_01 [{shapeType=Capsule, length=2.8703, radius=2.8265, center=<2.896 -0.0001 0>, rotation=<rot 90 0 0>}]
 dyn_ponytail_02 [{shapeType=Capsule, length=2.6448, radius=3.3435, center=<4.4673 0.0003 0.0001>, rotation=<rot 90 0 0>}]
 dyn_ponytail_03 [{shapeType=Capsule, length=3.7242, radius=2.8581, center=<4.5187 0 0>, rotation=<rot 90 0 0>}]
 dyn_ponytail_04 [{shapeType=Capsule, length=4.7946, radius=3.726, center=<4.9551 -0.0001 0>, rotation=<rot 90 0 0>}]
 dyn_ponytail_05 [{shapeType=Capsule, length=4.8024, radius=4.0956, center=<6.5817 0.0001 0>, rotation=<rot 80.0905 90.0002 -179.9998>}]
 dyn_ponytail_06 [{shapeType=Capsule, length=6.6168, radius=3.6728, center=<4.8212 0 -0.6388>, rotation=<rot 0.0001 0 83.467>}]
 dyn_ponytail_07 [{shapeType=Capsule, length=1.7316, radius=2.7807, center=<3.106 0.0001 0>, rotation=<rot 90 0 0>}]
 dyn_ponytail_08 [{shapeType=Capsule, length=2.5535, radius=2.2098, center=<2.9513 0 0>, rotation=<rot 90 0 0>}]
 dyn_ponytail_base [{shapeType=Capsule, length=0.125, radius=1.125, center=<0.0047 -0.0026 0.0453>}]
 dyn_ponytail_parent [{shapeType=Capsule, length=2.2461, radius=1.9626, center=<2.1477 0.0001 0>, rotation=<rot 90 0 0>}]
 dyn_ponytail_strand_l_01 [{shapeType=Capsule, length=0.9404, radius=1.0429, center=<1.267 0 0>, rotation=<rot 90 0 0>}]
 dyn_ponytail_strand_l_02 [{shapeType=Capsule, length=1.1989, radius=1.1531, center=<1.3757 0.0001 0>, rotation=<rot 90 0 0>}]
 dyn_ponytail_strand_l_03 [{shapeType=Capsule, length=2.5012, radius=1.135, center=<2.0421 0 0>, rotation=<rot 90 0 0>}]
 dyn_ponytail_strand_l_04 [{shapeType=Capsule, length=2.7743, radius=0.807, center=<1.857 0 0>, rotation=<rot 90 0 0>}]
 dyn_ponytail_strand_l_05 [{shapeType=Capsule, length=2.7962, radius=1.0531, center=<1.6583 0 0>, rotation=<rot 90 0 0>}]
 dyn_ponytail_strand_l_06 [{shapeType=Capsule, length=3.8789, radius=1.3016, center=<2.1563 0 0>, rotation=<rot 90 0 0>}]
 dyn_ponytail_strand_l_07 [{shapeType=Capsule, length=3.7003, radius=1.2173, center=<2.4828 0 0.0001>, rotation=<rot 90 0 0>}]
 dyn_ponytail_strand_mid_01 [{shapeType=Capsule, length=3.1595, radius=0.766, center=<2.0741 0 0.0001>, rotation=<rot 90 0 0>}]
 dyn_ponytail_strand_mid_02 [{shapeType=Capsule, length=1.9397, radius=0.75, center=<1.4103 0 -0.0001>, rotation=<rot 90 0 0>}]
 dyn_ponytail_strand_r_02 [{shapeType=Capsule, length=3.3401, radius=0.6724, center=<3.65 0.2531 0.2657>, rotation=<rot 0 95.0657 -84.445>}]
 dyn_ponytail_strand_r_03 [{shapeType=Capsule, length=3.4148, radius=0.7861, center=<2.2886 0 0>, rotation=<rot 90 0 0>}]
 dyn_ponytail_strand_r_04 [{shapeType=Capsule, length=1.701, radius=0.8136, center=<1.0253 0 0>, rotation=<rot 90 0 0>}]
 dyn_ponytail_strand_r_05 [{shapeType=Capsule, length=3.0126, radius=0.75, center=<1.6176 0 0>, rotation=<rot 90 0 0>}]
 head [{shapeType=Capsule, length=4, radius=7.875, center=<3.9295 1.1684 -0.0038>, rotation=<rot 90 -177.6141 -153.6272>}]
 neck_01 [{shapeType=Capsule, length=9.625, radius=5.875, center=<3.6406 2.3165 -0.0039>, rotation=<rot 90 20.556 31.2859>}]
 spine_05 [{shapeType=Capsule, length=9.125, radius=10.375, center=<-2.9699 0.2832 -0.0475>, rotation=<rot 90 0 0>}]
}
constraints [
 {bone1Name=dyn_ponytail_parent, bone2Name=dyn_ponytail_base, swing1LimitDegrees=10, swing2LimitDegrees=10, twistLimitDegrees=5}
 {bone1Name=dyn_ponytail_strand_l_01, bone2Name=dyn_ponytail_parent, swing1LimitDegrees=10, swing2LimitDegrees=10, twistLimitDegrees=5}
 {bone1Name=dyn_ponytail_strand_l_02, bone2Name=dyn_ponytail_strand_l_01, swing1LimitDegrees=10, swing2LimitDegrees=10, twistLimitDegrees=5}
 {bone1Name=dyn_ponytail_strand_l_03, bone2Name=dyn_ponytail_strand_l_02, swing1LimitDegrees=10, swing2LimitDegrees=10, twistLimitDegrees=5}
 {bone1Name=dyn_ponytail_strand_l_04, bone2Name=dyn_ponytail_strand_l_03, swing1LimitDegrees=10, swing2LimitDegrees=10, twistLimitDegrees=5}
 {bone1Name=dyn_ponytail_strand_l_05, bone2Name=dyn_ponytail_strand_l_04, swing1LimitDegrees=10, swing2LimitDegrees=10, twistLimitDegrees=5}
 {bone1Name=dyn_ponytail_strand_l_06, bone2Name=dyn_ponytail_strand_l_05, swing1LimitDegrees=10, swing2LimitDegrees=10, twistLimitDegrees=5}
 {bone1Name=dyn_ponytail_strand_l_07, bone2Name=dyn_ponytail_strand_l_06, swing1LimitDegrees=10, swing2LimitDegrees=10, twistLimitDegrees=5}
 {bone1Name=dyn_ponytail_01, bone2Name=dyn_ponytail_parent, swing1LimitDegrees=10, swing2LimitDegrees=10, twistLimitDegrees=5}
 {bone1Name=dyn_ponytail_02, bone2Name=dyn_ponytail_01, swing1LimitDegrees=10, swing2LimitDegrees=10, twistLimitDegrees=5}
 {bone1Name=dyn_ponytail_03, bone2Name=dyn_ponytail_02, swing1LimitDegrees=10, swing2LimitDegrees=10, twistLimitDegrees=5}
 {bone1Name=dyn_ponytail_04, bone2Name=dyn_ponytail_03, swing1LimitDegrees=10, swing2LimitDegrees=10, twistLimitDegrees=5}
 {bone1Name=dyn_ponytail_05, bone2Name=dyn_ponytail_04, swing1LimitDegrees=10, swing2LimitDegrees=10, twistLimitDegrees=5}
 {bone1Name=dyn_ponytail_06, bone2Name=dyn_ponytail_05, swing1LimitDegrees=10, swing2LimitDegrees=10, twistLimitDegrees=5}
 {bone1Name=dyn_ponytail_07, bone2Name=dyn_ponytail_06, swing1LimitDegrees=10, swing2LimitDegrees=10, twistLimitDegrees=5}
 {bone1Name=dyn_ponytail_08, bone2Name=dyn_ponytail_07, swing1LimitDegrees=10, swing2LimitDegrees=10, twistLimitDegrees=5}
 {bone1Name=dyn_ponytail_strand_mid_01, bone2Name=dyn_ponytail_04, swing1LimitDegrees=10, swing2LimitDegrees=10, twistLimitDegrees=3}
 {
  bone1Name=dyn_ponytail_strand_mid_02
  bone2Name=dyn_ponytail_strand_mid_01
  swing1LimitDegrees=10
  swing2LimitDegrees=10
  twistLimitDegrees=3
 }
 {bone1Name=dyn_ponytail_strand_r_03, bone2Name=dyn_ponytail_strand_r_02, swing1LimitDegrees=3, swing2LimitDegrees=3, twistLimitDegrees=3}
 {bone1Name=dyn_ponytail_strand_r_04, bone2Name=dyn_ponytail_strand_r_03, swing1LimitDegrees=3, swing2LimitDegrees=3, twistLimitDegrees=3}
 {bone1Name=dyn_ponytail_strand_r_05, bone2Name=dyn_ponytail_strand_r_04, swing1LimitDegrees=3, swing2LimitDegrees=3, twistLimitDegrees=3}
 {bone1Name=dyn_ponytail_strand_r_02, bone2Name=dyn_ponytail_03, swing1LimitDegrees=3, swing2LimitDegrees=3, twistLimitDegrees=2}
 {
  bone1Name=clavicle_r
  bone2Name=spine_05
  swing1LimitDegrees=45
  swing1Motion=Free
  swing2LimitDegrees=45
  swing2Motion=Free
  twistLimitDegrees=45
  twistMotion=Free
 }
]
