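# Animation Blueprint for the Donut Plate female cosmetic's face accessory mesh, adding secondary motion to the ponytail and bangs.
# It copies the attached parent's pose, then runs RigidBody simulations relative to spine_05 for the ponytail and, at partial strength, the bangs.
# When bIsInFrontEnd is true, an additive offset adjusts the ponytail underside bone.
# Two AnimDynamics ponytail chains are left unconnected, and the EventGraph has only unwired stub nodes.
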AnimBlueprint /BRCosmetics/Characters/Player/Female/Medium/Bodies/F_MED_Donut_Plate/Meshes/Parts/F_MED_Donut_Plate_FaceAcc_AnimBp
target_skeleton=&/F_MED_Donut_Plate_FaceAcc_Skeleton
graphs [
 AnimGraph {nodes [
   "Anim Dynamics - Chain: dyn_ponytail_strand_l_03 -> dyn_ponytail_strand_l_07"(Alpha=1, ExternalForce="0,0,0") {
    angularDampingOverride=0.9999
    bChain=true
    bGravityOverrideInSimSpace=true
    bOverrideLinearDamping=true
    gravityScale=0.99
    linearDampingOverride=0.9999
    numSolverIterationsPostUpdate=10
    numSolverIterationsPreUpdate=10
    alphaBoolBlend {blendOption=Linear}
    boundBone {boneName=dyn_ponytail_strand_l_03}
    chainEnd {boneName=dyn_ponytail_strand_l_07}
    componentLinearAccScale=<0.25 0.25 0.25>
    componentLinearVelScale=<0.25 0.25 0.25>
    gravityOverride=<1 1 1>
    physicsBodyDefinitions [{
      sphereCollisionRadius=0
      boundBone {boneName=dyn_ponytail_strand_l_03}
      boxExtents=<8 8 6>
      constraintSetup {angularLimitsMax=<0 10 20>, angularLimitsMin=<0 -10 -10>, angularTarget=<0 0 0>}
      localJointOffset=<2 0 0>
     }]
    relativeSpaceBone {boneName=root}
   }
   "Anim Dynamics - Chain: dyn_ponytail_02 -> dyn_ponytail_08"(Alpha=1, ExternalForce="0,0,0") {
    angularDampingOverride=0.9999
    bChain=true
    bGravityOverrideInSimSpace=true
    bOverrideLinearDamping=true
    gravityScale=0.1
    linearDampingOverride=0.9999
    numSolverIterationsPostUpdate=10
    numSolverIterationsPreUpdate=10
    alphaBoolBlend {blendOption=Linear}
    boundBone {boneName=dyn_ponytail_02}
    chainEnd {boneName=dyn_ponytail_08}
    componentLinearAccScale=<0.25 0.25 0.25>
    componentLinearVelScale=<0.25 0.25 0.25>
    gravityOverride=<0.999 0.999 0.999>
    physicsBodyDefinitions [{
      sphereCollisionRadius=0
      boundBone {boneName=dyn_ponytail_02}
      boxExtents=<8 8 6>
      constraintSetup {angularLimitsMax=<0 10 5>, angularLimitsMin=<0 -10 -20>, angularTarget=<0 0 0>}
      localJointOffset=<2 0 0>
     }]
    relativeSpaceBone {boneName=root}
   }
   n1 = RigidBody(ComponentPose=Local To Component(LocalPose=Copy Pose From Mesh() {bUseAttachedParent=true})) {
    bClampLinearTranslationLimitToRefPose=true
    bOverrideWorldGravity=true
    overridePhysicsAsset=&/F_MED_Donut_Plate_Ponytail_FaceAcc_Physics
    simulationSpace=BaseBoneSpace
    alphaBoolBlend {blendOption=Linear}
    alphaScaleBiasClamp {bInterpResult=true, interpSpeedDecreasing=5, interpSpeedIncreasing=6}
    baseBoneRef {boneName=spine_05}
    componentAppliedLinearAccClamp=<3000 3000 3000>
    componentLinearAccScale=<0.25 0.25 0.25>
    componentLinearVelScale=<0.25 0.25 0.25>
    overrideWorldGravity {z=-980}
   }
   n2 = RigidBody(ComponentPose=@n1, Alpha=0.75) {
    bClampLinearTranslationLimitToRefPose=true
    bOverrideWorldGravity=true
    overridePhysicsAsset=&/F_MED_Donut_Plate_FaceAcc_Bangs_Physics
    simulationSpace=BaseBoneSpace
    alphaBoolBlend {blendOption=Linear}
    alphaScaleBiasClamp {bInterpResult=true, interpSpeedDecreasing=5, interpSpeedIncreasing=6}
    baseBoneRef {boneName=spine_05}
    componentAppliedLinearAccClamp=<6000 6000 6000>
    componentLinearAccScale=<0.25 0.25 0.25>
    componentLinearVelScale=<0.25 0.25 0.25>
    overrideWorldGravity {z=-980}
   }
   n3 = "Transform (Modify) Bone - Bone: dyn_ponytail_underside_01"!(ComponentPose=@n2, bAlphaBoolEnabled:bool=Get bIsInFrontEnd(), Translation="0,0,-1", Rotation="-15,0,0", Scale="1,1,1") -> (Pose) {
    alphaInputType=Bool
    rotationMode=BMM_Additive
    rotationSpace=BCS_BoneSpace
    translationMode=BMM_Additive
    translationSpace=BCS_BoneSpace
    alphaBoolBlend {blendOption=Linear}
    boneToModify {boneName=dyn_ponytail_underside_01}
    rotation {pitch=-15}
    translation {z=-1}
   }
   Output Pose(Result=Component To Local(ComponentPose=@n3)) {layerGroup=None}
  ]}
 EventGraph {nodes [
   TryGetPawnOwner()
   Event BlueprintUpdateAnimation()
  ]}
]
variables []
